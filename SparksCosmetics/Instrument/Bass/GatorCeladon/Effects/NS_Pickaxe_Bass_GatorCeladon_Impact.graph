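# Niagara impact effect for the GatorCeladon bass pickaxe, bursting warm yellow-orange flashes, flares with a light, sparks, streaks, spikes and smoke.
# Most directional emitters launch particles along a system direction vector and scale with a system global scale.
# Mesh debris picks velocity, cone and gravity from the physical-surface user parameter for one spawn group.
# Front-end flags adjust counts, sizes, speeds and brightness; the second directional-sparks emitter is disabled.
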NiagaraSystem /SparksCosmetics/Instrument/Bass/GatorCeladon/Effects/NS_Pickaxe_Bass_GatorCeladon_Impact
dependencies {
 usedDataInterfaces [
  &NiagaraDataInterfaceCurve
  &NiagaraDataInterfaceVectorCurve
  &NiagaraDataInterfaceVector2DCurve
  &NiagaraDataInterfaceArrayFloat
  &NiagaraDataInterfaceColorCurve
  &NiagaraDataInterfaceArrayFloat3
 ]
 usedDynamicInputs [
  &Vector2DFromFloat
  /Niagara/DynamicInputs/Math/Clamp_Vector
  &SelectVectorFromArray
 ]
 usedModules [
  &self:SystemSpawnScript.SystemScriptSource.SystemScriptGraph.NiagaraNodeAssignment_1.SetVariables_#1
  &SystemState
  &self:SystemSpawnScript.SystemScriptSource.SystemScriptGraph.NiagaraNodeAssignment_0.SetVariables_#2
 ]
}
emitters {
 DirectionalSparks001 {
  bEnabled=false
  emitterUpdateScript [
   &EmitterState()
   &SpawnBurst_Instantaneous(
    Spawn Count=20
    Spawn Probability=&MakeCustomFloatFromBool(Bool=$User.bIsFrontEndPreview, True Float=0, False Float=1)
    Use Spawn Probability=true
    UseLoopCountLimit=false
   )
   &SpawnBurst_Instantaneous(
    Spawn Count=9
    Spawn Group=1
    Spawn Probability=&MakeCustomFloatFromBool(Bool=$User.bIsFrontEndPreview, True Float=0, False Float=1)
    Use Spawn Probability=true
    UseLoopCountLimit=false
   ) {module=SpawnBurst_Instantaneous001}
  ]
  particleSpawnScript [
   &InitializeParticle(
    Color Maximum=<rgba 2 0.3214 0 1>
    Color Minimum=<rgba 0.966 1.25 0 1>
    Color Mode=::NewEnumerator2
    Lifetime=&FloatFromCurve(
     FloatCurve=DataInterface() {propertyValues {Curve {keys [
         {arriveTangent=-5.4544, interpMode=RCIM_Cubic, leaveTangent=-5.4544, tangentMode=RCTM_User, time=0, value=0.75}
         {arriveTangent=-0.65, interpMode=RCIM_Cubic, leaveTangent=-0.65, time=0.1, value=0.4}
         <at 1 0.1>
        ]}}}
     CurveIndex=&RandomRangeFloat(Minimum=0, Maximum=1)
    )
    Sprite Size Mode=::NewEnumerator3
    Uniform Sprite Size=&Multiply_Float(A=$System.Global Scale, B=&RandomRangeFloat(Minimum=8, Maximum=14))
    UsePositionOffset=false
    Write Material Random=true
   )
   &ShapeLocation(
    MaskSpawnGroup=false
    Override Seed=false
    Rotation Axis=$System.DirectionVector
    Rotation Mode=::NewEnumerator1
    Sphere Radius=5
    UseOwnerScale=true
   )
   &CameraOffset(Camera Offset Amount=10)
   &AddVelocityFromPoint(
    Module.Use Falloff Distance=false
    Velocity Strength=&Multiply_Float(
     A=&RandomRangeFloat(Minimum=150, Maximum=200)
     B=&FN_FrontEndFloatModifier(Gameplay=&MaskFloatBySpawnGroup(Masked Float=1, Passthrough Float=3), Front End=$User.FrontEndScale)
    )
   )
   &AddVelocity(
    Velocity=$System.DirectionVector
    Velocity Speed Scale=&Multiply_Float(A=&RandomRangeFloat(Minimum=200, Maximum=500), B=$System.Global Scale)
   )
  ]
  particleUpdateScript [
   &ParticleState()
   &AccelerationForce(Acceleration=&RandomRangeVector(Minimum=<0 0 -2000>, Maximum=<0 0 -1000>))
   &Drag(Drag=&RandomRangeFloat(Minimum=1, Maximum=4), Use Linear Drag=true, Use Rotational Drag=true)
   &SolveForcesAndVelocity(Clamp Velocity=false, Limit Acceleration=false)
   &ScaleSpriteSizeBySpeed(Max Scale Factor=<0.1 2>, Min Scale Factor=<0.25 0.2>)
   &ScaleSpriteSize(Uniform Curve Sprite Scale=DataInterface() {propertyValues {Curve {keys [{arriveTangent=-6.725, interpMode=RCIM_Cubic, leaveTangent=-6.725, tangentMode=RCTM_User, time=0.65, value=1}, <at 1 0 RCIM_Cubic>]}}})
   &ScaleColor(
    Scale Alpha=&FloatFromCurve(
     FloatCurve=DataInterface() {propertyValues {Curve {keys [
         {interpMode=RCIM_Cubic, tangentMode=RCTM_User, time=0, value=1}
         {arriveTangent=-0.3845, interpMode=RCIM_Cubic, leaveTangent=-0.3845, tangentMode=RCTM_User, time=0.5, value=0.9}
         {arriveTangent=-0.2828, interpMode=RCIM_Cubic, leaveTangent=-0.2828, tangentMode=RCTM_User, time=1, value=0}
        ]}}}
     Scale Curve=&RandomRangeFloat(Minimum=0.25, Maximum=5)
    )
    Scale RGB=&Multiply_VectorByFloat(
     Vector=&MakeVectorFromLinearColorRGB(Color=&ColorFromCurve(ColorCurve=DataInterface() {propertyValues {
         AlphaCurve {keys [<at 0 1 RCIM_Cubic>]}
         BlueCurve {keys [<at 0.1873 2 RCIM_Cubic>,<at 0.854 0.0935 RCIM_Cubic>]}
         GreenCurve {keys [<at 0.1873 2 RCIM_Cubic>,<at 0.854 0 RCIM_Cubic>]}
         RedCurve {keys [<at 0.1873 2 RCIM_Cubic>,<at 0.854 0.6406 RCIM_Cubic>]}
        }}))
     Float=20
    )
    ScaleA=true
    ScaleRGB=true
   )
  ]
  rendererClasses [&NiagaraSpriteRendererProperties]
  renderers [{rendererClass=&NiagaraSpriteRendererProperties}]
 }
 FN_PhysMatDebris001 {
  emitterUpdateScript [
   &EmitterState()
   &SpawnBurst_Instantaneous(
    Spawn Count=&RandomRangeInt(Minimum=4, Maximum=6)
    Use Spawn Probability=false
    UseLoopCountLimit=false
   )
   &SpawnBurst_Instantaneous(
    Spawn Count=&MakeCustomIntFromBool(Bool=$User.bIsFrontEnd, True INT=0, False INT=&RandomRangeInt(Minimum=3, Maximum=4))
    Spawn Group=1
    Use Spawn Probability=false
    UseLoopCountLimit=false
   ) {module=SpawnBurst_Instantaneous001}
  ]
  particleSpawnScript [
   /SetVariables(Particles.BiasedRandom=&FloatFromCurve(
     FloatCurve=DataInterface() {propertyValues {Curve {keys [
         <at 0.2 1 RCIM_Cubic>
         {arriveTangent=-3.3394, interpMode=RCIM_Cubic, leaveTangent=-3.3394, tangentMode=RCTM_User, time=0.5, value=0.5}
         <at 0.8 0 RCIM_Cubic>
        ]}}}
     CurveIndex=&RandomRangeFloat(Minimum=0, Maximum=1)
    )) {isSetParameters=true}
   &InitializeParticle(
    Color Mode=::NewEnumerator0
    Lifetime=&RandomRangeFloat(Minimum=0.5, Maximum=1.25)
    Mass=&Lerp_Float(A=0.6, B=1.2, Alpha=$Particles.BiasedRandom)
    Mass Mode=::NewEnumerator1
    Mesh Scale=&MaskVectorBySpawnGroup(
     Masked Vector=&Multiply_VectorByFloat(
      Vector=&RandomRangeVector(Minimum=<0.9 0.4 0.5>, Maximum=<1 1 1>)
      Float=&Lerp_Float(A=0.7, B=1, Alpha=$Particles.BiasedRandom)
     )
     Passthrough Vector=&Multiply_VectorByFloat(
      Vector=&RandomRangeVector(Minimum=<0.9 0.4 0.5>, Maximum=<1 1 1>)
      Float=&Lerp_Float(A=0.4, B=0.7, Alpha=$Particles.BiasedRandom)
     )
    )
    Mesh Scale Mode=::NewEnumerator4
    UsePositionOffset=false
    Write Material Random=true
   )
   &FN_PhysicalSurface(
    Dirt Mesh Count=2
    Glass Mesh Count=3
    Ice Mesh Count=2
    Metal Gravity=&RandomRangeFloat(Minimum=-350, Maximum=-1800)
    Metal Mesh Count=2
    Metal Velocity=&RandomRangeFloat(Minimum=80, Maximum=375)
    Metal Velocity Cone Angle=130
    Physical Surface=&MaskIntBySpawnGroup(Masked Int=$User.PhysicalSurface, Passthrough Int=1, Spawn Group Mask=0)
    Stone Gravity=&RandomRangeFloat(Minimum=-700, Maximum=-1250)
    Stone Mesh Count=2
    Stone Velocity=&RandomRangeFloat(Minimum=80, Maximum=550)
    Stone Velocity Cone Angle=120
    Wood Color:LinearColor=&MaskLinearColorBySpawnGroup(Masked Linear Color=<rgba 1 1 1 1>, Passthrough Linear Color=<rgba 1.5 0.7582 0.5038 1>)
    Wood Gravity=&RandomRangeFloat(Minimum=-450, Maximum=-900)
    Wood Mesh Count=2
    Wood Scale=1.5
    Wood Velocity=&RandomRangeFloat(Minimum=80, Maximum=600)
    Wood Velocity Cone Angle=130
   )
   &AddVelocityInCone(
    Cone Angle=$Output.FN_PhysicalSurface.SurfaceResponseConeAngle
    Module.Use Velocity Falloff On Cone Axis=false
    Velocity Strength=$Output.FN_PhysicalSurface.SurfaceResponseVelocity
   )
   &InitialMeshOrientation(Use Rotation Vector=false)
   &AddRotationalVelocity(Rotation Rate=&RandomVector(Vector Scale=&RandomRangeFloat(Minimum=1, Maximum=2)))
   &ApplyInitialForces()
   /SetVariables(Particles.GravityOverrideValue=$Output.FN_PhysicalSurface.GravityOverride) {isSetParameters=true}
  ]
  particleUpdateScript [
   &ParticleState()
   &GravityForce(Gravity=$Particles.GravityOverrideValue)
   &Collision(
    Advanced Aging Rate=1
    Age Colliding Particles=false
    CPU Collision Trace Channel=ECC_WorldStatic
    Collision Enabled=&Boolean_AndOperation(
     Bool1=&SetBoolByFloatComparison(A=$Particles.Initial.Lifetime, B=1)
     Bool2=&SetBoolByFloatComparison(A=$Particles.Age, B=0.12)
    )
    Control Roll On Collision=true
    Friction=1
    Friction During a Bounce=0.25
    Max CPU Trace Length=100
    Module.Control Rest State Via Surface Normal=false
    Module.EnableMaxCollisionCount=false
    Module.Randomize Collision Normal Vector=true
    Randomize Collision Normal=0.25
    ScaleRateOfRotation=false
   ) {enabled=false}
   &SolveForcesAndVelocity(Clamp Velocity=false, Limit Acceleration=false)
   &ScaleMeshSize(Scale Factor=&VectorFromFloat(Value=&FloatFromCurve(FloatCurve=DataInterface() {propertyValues {Curve {keys [<at 0.6 1 RCIM_Cubic>, {arriveTangent=-6.308, interpMode=RCIM_Cubic, leaveTangent=-6.308, tangentMode=RCTM_User, time=1, value=0}]}}})))
   &DynamicMaterialParameters(Module.Param0WriteEnabled=true, Module.Param1WriteEnabled=true, Module.Param2WriteEnabled=false, Module.Param3WriteEnabled=false)
  ]
  rendererClasses [&NiagaraSpriteRendererProperties, &NiagaraMeshRendererProperties]
  renderers [
   {rendererClass=&NiagaraSpriteRendererProperties}
   {rendererClass=&NiagaraMeshRendererProperties, rendererIndex=1}
   {rendererClass=&NiagaraMeshRendererProperties, rendererIndex=2}
   {rendererClass=&NiagaraMeshRendererProperties, rendererIndex=3}
   {rendererClass=&NiagaraMeshRendererProperties, rendererIndex=4}
   {rendererClass=&NiagaraMeshRendererProperties, rendererIndex=5}
   {rendererClass=&NiagaraMeshRendererProperties, rendererIndex=6}
   {rendererClass=&NiagaraMeshRendererProperties, rendererIndex=7}
   {rendererClass=&NiagaraMeshRendererProperties, rendererIndex=8}
  ]
 }
 Flash_Combined {
  emitterUpdateScript [
   &EmitterState()
   &SpawnBurst_Instantaneous(Use Spawn Probability=false, UseLoopCountLimit=false)
   &SpawnBurst_Instantaneous(Spawn Group=1, Use Spawn Probability=false, UseLoopCountLimit=false) {module=SpawnBurst_Instantaneous001}
  ]
  particleSpawnScript [
   /SetVariables(
    Particles.MeshOrientation=&MakeQuatFromAxisAngle(
     Axis=&RandomRangeVector(Minimum=<0 0 0>, Maximum=<1 1 1>, Evaluation Type=::NewEnumerator1)
     AngleInDegrees=90
    )
    Particles.VisibilityTag=$Emitter.SpawnGroup
   ) {isSetParameters=true}
   &InitializeParticle(
    Lifetime=&MaskFloatBySpawnGroup(Masked Float=&FN_FrontEndFloatModifier(Gameplay=0.22, Front End=0.25), Passthrough Float=0.15)
    Mesh Scale=&Multiply_VectorByFloat(Vector=&RandomRangeVector(Minimum=<0.8 1.6 1.6>, Maximum=<1.1 2.2 2.2>), Float=0.9)
    Mesh Scale Mode=::NewEnumerator4
    Sprite Rotation Mode=::NewEnumerator2
    Sprite Size=&Multiply_Vector2D_ByFloat(Float=&FN_FrontEndFloatModifier(Gameplay=1, Front End=$System.Global Scale), Vector2D=<180 180>)
    Sprite Size Mode=::NewEnumerator4
    UsePositionOffset=false
    Write Material Random=true
   )
   /SetVariables(
    Particles.FlareCardColor:LinearColor=$System.TOD
    Particles.FlareCardDynamicParameter:Vector4f=&MakeVector4(X=0.2, Y=1, Z=0, W=0)
   ) {isSetParameters=true}
   &FN_PerParticleLOD(Enable SpriteSize Distance Scaling=true, Sprite Size Scale Far=<2.5 2.5>)
   &CameraOffset(Camera Offset Amount=15)
  ]
  particleUpdateScript [
   &ParticleState()
   &SolveForcesAndVelocity(Clamp Velocity=false, Limit Acceleration=false) {enabled=false}
   &DynamicMaterialParameters(
    Index 0 Param 1=&FloatFromCurve(FloatCurve=DataInterface() {propertyValues {Curve {
        defaultValue=0
        keys [
         {interpMode=RCIM_Cubic, tangentMode=RCTM_User, time=0, value=0}
         {arriveTangent=2.3262, interpMode=RCIM_Cubic, leaveTangent=2.3262, tangentMode=RCTM_User, time=1, value=1}
        ]
       }}})
    Index 0 Param 2=45
    Index 0 Param 3=1
    Index 0 Param 4=1
    Module.Index0Float0Write=true
    Module.Index0Float1Write=true
    Module.Index0Float2Write=true
    Module.Index0Float3Write=true
   )
   &ScaleColor(
    Scale Alpha=&FloatFromCurve(FloatCurve=DataInterface() {propertyValues {Curve {keys [<at 0 1>,<at 1 0>]}}})
    Scale RGB=&VectorFromCurve(VectorCurve=DataInterface() {propertyValues {XCurve {keys [<at 0 1>,<at 1 1>]}, YCurve {keys [<at 0 1>,<at 1 1>]}, ZCurve {keys [<at 0 1>,<at 1 1>]}}})
    ScaleA=true
    ScaleRGB=true
   )
   /SetVariables(Particles.UpdateFlareCardColor:LinearColor=&MakeLinearColorFromVectorAndFloat(
     "Vector (RGB)"=&Multiply_Vector(
      A=&Multiply_Vector(
       A=&MakeVectorFromLinearColorRGB(Color=$Particles.Initial.FlareCardColor)
       B=&MakeVectorFromLinearColorRGB(Color=&ColorFromCurve(ColorCurve=DataInterface() {propertyValues {
           AlphaCurve {keys [<at 0 1 RCIM_Cubic>,<at 1 0 RCIM_Cubic>]}
           BlueCurve {keys [<at 0.0198 0>,<at 0.491 0>,<at 1 0>]}
           GreenCurve {keys [<at 0.0198 1.2233>,<at 0.491 0.9522>,<at 1 0.4993>]}
           RedCurve {keys [<at 0.0198 2>,<at 0.491 2>,<at 1 3>]}
          }}))
      )
      B=&VectorFromFloat(Value=&FN_FrontEndFloatModifier(Gameplay=1.5, Front End=1))
     )
     "Float (Alpha)"=&FloatFromCurve(FloatCurve=DataInterface() {propertyValues {Curve {keys [{arriveTangent=-1.9727, interpMode=RCIM_Cubic, time=0.1, value=1}, <at 1 0 RCIM_Cubic>]}}})
    )) {isSetParameters=true}
   &ScaleMeshSize(Scale Factor=&VectorFromFloat(Value=&FloatFromCurve(
      FloatCurve=DataInterface() {propertyValues {Curve {keys [
          {arriveTangent=0.3871, interpMode=RCIM_Cubic, time=0, value=0.5}
          {arriveTangent=0.6012, interpMode=RCIM_Cubic, leaveTangent=0.6012, tangentMode=RCTM_User, time=0.7128, value=0.698}
          {arriveTangent=1.7425, interpMode=RCIM_Cubic, leaveTangent=1.7425, tangentMode=RCTM_User, time=1, value=1}
         ]}}}
      Scale Curve=&Multiply_Float(A=1.3, B=$System.Global Scale)
     )))
  ]
  rendererClasses [&NiagaraSpriteRendererProperties, &NiagaraMeshRendererProperties]
  renderers [
   {rendererClass=&NiagaraSpriteRendererProperties}
   {rendererClass=&NiagaraMeshRendererProperties, rendererIndex=1}
  ]
 }
 Flash_Flare {
  emitterUpdateScript [
   &EmitterState(Inactive Response=::NewEnumerator1, Life Cycle Mode=::NewEnumerator1, Loop Behavior=::NewEnumerator1, Loop Duration=1, UseLoopDelay=false)
   &SpawnBurst_Instantaneous(Use Spawn Probability=true, UseLoopCountLimit=false)
  ]
  particleSpawnScript [
   &InitializeParticle(
    Color=<rgba 3 1.3805 0 1>
    Lifetime=0.18
    Sprite Size=&Multiply_Vector2D_ByFloat(Float=0.8, Vector2D=<80 240>)
    Sprite Size Mode=::NewEnumerator4
    UsePositionOffset=false
    Write Material Random=true
   )
   &CameraOffset(Camera Offset Amount=30)
   &DynamicMaterialParameters(
    Index 0 Param 1=0.2
    Index 0 Param 2=1
    Index 3 Param 1=1.5
    Index 3 Param 2=1
    Module.Param0WriteEnabled=true
    Module.Param0WriteEnabled3=true
    Module.Param1WriteEnabled=true
    Module.Param1WriteEnabled3=true
    Module.Param2WriteEnabled=true
    Module.Param2WriteEnabled3=true
    Module.Param3WriteEnabled=true
    Module.Param3WriteEnabled3=true
    Write Parameter Index 3=true
   )
   /SetVariables(Particles.NewAlignment:Vector3f=<1 0 0>) {isSetParameters=true}
   &AddVelocity(
    Velocity=$System.DirectionVector
    Velocity Speed Scale=&Multiply_Float(
     A=&RandomRangeFloat(Minimum=200, Maximum=300)
     B=&FN_FrontEndFloatModifier(Gameplay=1, Front End=1.1)
    )
   )
  ]
  particleUpdateScript [
   &ParticleState()
   &ScaleSpriteSize(
    Non-Uniform Curve Scale=&Multiply_Vector2D_ByFloat(Float=$System.Global Scale, Vector2D=<1 1>)
    Non-Uniform Curve Sprite Scale=DataInterface() {propertyValues {
      XCurve {keys [{tangentMode=RCTM_User, time=0, value=0}, <at 0.2 1.075>, {tangentMode=RCTM_User, time=1, value=-0.1}]}
      YCurve {keys [{tangentMode=RCTM_User, time=0, value=0}, <at 0.1 1.025>, {tangentMode=RCTM_User, time=1, value=0.8}]}
     }}
    Scale Sprite Size Mode=::NewEnumerator3
   )
   &SolveForcesAndVelocity(Clamp Velocity=false, Limit Acceleration=false)
  ]
  rendererClasses [&NiagaraSpriteRendererProperties]
  renderers [{rendererClass=&NiagaraSpriteRendererProperties}]
 }
 Flash_Flare_Light {
  emitterUpdateScript [
   &EmitterState()
   &SpawnBurst_Instantaneous(Use Spawn Probability=true, UseLoopCountLimit=false)
  ]
  particleSpawnScript [
   &InitializeParticle(
    Color=<rgba 1.25 0.6234 0 1>
    Lifetime=0.25
    Sprite Size Mode=::NewEnumerator3
    Uniform Sprite Size=120
    UsePositionOffset=false
    Write Material Random=true
   )
   &SystemLocation(MaskSpawnGroup=false, Offset=$System.HitOffset)
   &AddVelocity(Velocity=$System.DirectionVector)
  ]
  particleUpdateScript [
   &ParticleState()
   &ScaleSpriteSize(
    Uniform Curve Scale=$System.Global Scale
    Uniform Curve Sprite Scale=DataInterface() {propertyValues {Curve {keys [
        {arriveTangent=3.5908, interpMode=RCIM_Cubic, leaveTangent=3.5908, tangentMode=RCTM_User, time=0, value=0}
        {tangentMode=RCTM_User, time=0.8, value=1}
       ]}}}
   )
   &ScaleColor(
    Scale Alpha=&FloatFromCurve(FloatCurve=DataInterface() {propertyValues {Curve {keys [
         {arriveTangent=-5.6669, interpMode=RCIM_Cubic, leaveTangent=-5.6669, tangentMode=RCTM_User, time=0, value=2}
         {arriveTangent=-1.179, interpMode=RCIM_Cubic, leaveTangent=-1.179, tangentMode=RCTM_User, time=0.5, value=0.5}
         {arriveTangent=-0.8686, interpMode=RCIM_Cubic, leaveTangent=-0.8686, tangentMode=RCTM_User, time=1, value=0}
        ]}}})
    Scale RGB=&VectorFromFloat(Value=&FN_FrontEndFloatModifier(Gameplay=1.5, Front End=1))
    ScaleA=true
    ScaleRGB=true
   )
   &FN_Light_Attributes(
    Color=$Particles.Color
    Create Color Attribute=true
    Create Exponent Attribute=true
    Create Position Attribute=true
    Create Radius Attribute=true
    Exponent=5
    Position=<15 0 0>
    Radius=150
    Scale Color=&FloatFromCurve(
     FloatCurve=DataInterface() {propertyValues {Curve {keys [
         {interpMode=RCIM_Cubic, tangentMode=RCTM_User, time=0, value=1}
         {arriveTangent=-0.5426, interpMode=RCIM_Cubic, leaveTangent=-0.5426, tangentMode=RCTM_User, time=1, value=0}
        ]}}}
     Scale Curve=5
    )
   )
   &CameraOffset(Camera Offset Amount=&Lerp_Float(A=15, B=4, Alpha=&MakeCustomFloatFromBool(Bool=$User.bIsFrontEnd, True Float=1, False Float=0)))
   /SetVariables(
    Particles.DarkBackerColor:LinearColor=&MultiplyLinearColor(
     Linear Color=<rgba 0.1667 0.0615 0.0011 1>
     Scale Factor=&MakeVector4FromVectorAndFloat(
      "Vector (RGB)"=&VectorFromFloat(Value=&Lerp_Float(A=1, B=3, Alpha=$Particles.NormalizedAge))
      "Float (Alpha)"=&FloatFromCurve(FloatCurve=DataInterface() {propertyValues {Curve {keys [
           {arriveTangent=-17.4607, interpMode=RCIM_Cubic, leaveTangent=-17.4607, tangentMode=RCTM_User, time=0, value=5}
           {arriveTangent=-2.7181, tangentMode=RCTM_User, time=0.8, value=0}
          ]}}})
     )
    )
    Particles.DarkBackerOffset:NiagaraFloat=&Add_Float(A=$Particles.CameraOffset, B=-2)
    Particles.DarkBackerScale:Vector2f=&Multiply_Vector2D(A=$Particles.SpriteSize, B=<3.2 1.8>)
    Particles.DarkBackerVelocity:Vector3f=$System.DirectionVector
    Particles.DarkBackerVisibility:NiagaraInt32=&SpawnGroup()
   ) {isSetParameters=true}
   &SolveForcesAndVelocity(Clamp Velocity=false, Limit Acceleration=false)
  ]
  rendererClasses [&NiagaraSpriteRendererProperties, &NiagaraLightRendererProperties]
  renderers [
   {rendererClass=&NiagaraSpriteRendererProperties}
   {rendererClass=&NiagaraLightRendererProperties, rendererIndex=1}
   {rendererClass=&NiagaraSpriteRendererProperties, rendererIndex=2}
  ]
 }
 Impact_Smoke {
  emitterUpdateScript [
   &EmitterState()
   &SpawnBurst_Instantaneous(
    Spawn Count=&RandomRangeInt(Minimum=2, Maximum=3)
    Use Spawn Probability=false
    UseLoopCountLimit=false
   )
  ]
  particleSpawnScript [
   &InitializeParticle(
    Color Maximum=<rgba 1 0.8251 0.9153 0.4>
    Color Minimum=<rgba 1 0.7984 0.7307 0.2>
    Color Mode=::NewEnumerator2
    Lifetime Max=0.5
    Lifetime Min=0.3
    Lifetime Mode=::NewEnumerator1
    Sprite Size Mode=::NewEnumerator3
    Uniform Sprite Size=&Multiply_Float(A=$System.Global Scale, B=&RandomRangeFloat(Minimum=100, Maximum=130))
    UsePositionOffset=false
    Write Material Random=true
   )
   &CameraOffset(Camera Offset Amount=&RandomRangeFloat(Minimum=40, Maximum=50, Evaluation Type=::NewEnumerator1))
   &ShapeLocation(
    Hemisphere Distribution=<1 0.5>
    MaskSpawnGroup=false
    Override Seed=false
    Rotation Angle=90
    Rotation Mode=::NewEnumerator1
    Sphere Radius=&MakeCustomFloatFromBool(Bool=$User.bIsFrontEnd, True Float=5, False Float=10)
    UseOwnerScale=true
   )
   &AddVelocity(Velocity=&Multiply_VectorByFloat(Vector=$Particles.ShapeLocation.ShapeVector, Float=10))
  ]
  particleUpdateScript [
   &ParticleState()
   &ScaleSpriteSize(Uniform Curve Sprite Scale=DataInterface() {propertyValues {Curve {keys [
        {arriveTangent=4.2019, interpMode=RCIM_Cubic, leaveTangent=4.2019, tangentMode=RCTM_User, time=0, value=0}
        {arriveTangent=0.6506, interpMode=RCIM_Cubic, leaveTangent=0.6506, tangentMode=RCTM_User, time=0.3745, value=0.8147}
        {interpMode=RCIM_Cubic, tangentMode=RCTM_User, time=1, value=1}
       ]}}})
   &ScaleColor(
    Scale Alpha=&FloatFromCurve(
     FloatCurve=DataInterface() {propertyValues {Curve {keys [
         {arriveTangent=2.4358, interpMode=RCIM_Cubic, leaveTangent=2.4358, tangentMode=RCTM_User, time=0, value=0}
         <at 0.25 0.5 RCIM_Cubic>
         <at 1 0 RCIM_Cubic>
        ]}}}
     Scale Curve=&FN_FrontEndFloatModifier(Gameplay=1, Front End=0.2)
    )
    Scale RGB=&VectorFromCurve(VectorCurve=DataInterface() {propertyValues {XCurve {keys [<at 0 1>,<at 1 1>]}, YCurve {keys [<at 0 1>,<at 1 1>]}, ZCurve {keys [<at 0 1>,<at 1 1>]}}})
    ScaleA=true
    ScaleRGB=true
   )
   &DynamicMaterialParameters(
    Index 0 Param 1=&Add_Float(
     A=&RandomRangeFloat(Minimum=0, Maximum=1)
     B=&FloatFromCurve(
      FloatCurve=DataInterface() {propertyValues {Curve {keys [
          <at 0 0 RCIM_Cubic>
          {arriveTangent=1.5947, interpMode=RCIM_Cubic, leaveTangent=1.5947, tangentMode=RCTM_User, time=0.2557, value=0.5257}
          <at 1 1 RCIM_Cubic>
         ]}}}
      CurveIndex=$Particles.Age
      Scale Curve=0.25
     )
    )
    Module.Param0WriteEnabled=true
    Module.Param1WriteEnabled=true
    Module.Param2WriteEnabled=true
    Module.Param3WriteEnabled=true
   ) {module=DynamicMaterialParameters001}
   &SolveForcesAndVelocity(Clamp Velocity=false, Limit Acceleration=false)
  ]
  rendererClasses [&NiagaraSpriteRendererProperties]
  renderers [
   {rendererClass=&NiagaraSpriteRendererProperties}
   {rendererClass=&NiagaraSpriteRendererProperties, rendererIndex=1}
  ]
 }
 Sparks {
  emitterUpdateScript [
   &EmitterState(
    Enable Distance Culling=true
    Max Distance Response=::KillImmediately
    Module.bApplySpawnCountScaleByDistance=true
    Module.bUseMaxDistance=true
    Module.bUseMinDistance=false
    Scalability Mode=::NewEnumerator1
    Scale Spawn Count=true
    Spawn Count Scale By Distance=DataInterface() {propertyValues {Curve {keys [<at 0.3676 0.9657>,<at 0.9962 0.4982>]}}}
   )
   &SpawnBurst_Instantaneous(Spawn Count=29, Use Spawn Probability=false, UseLoopCountLimit=false)
  ]
  particleSpawnScript [
   &InitializeParticle(
    Color Maximum=<rgba 2 0.3214 0 1>
    Color Minimum=<rgba 0.966 1.25 0 1>
    Color Mode=::NewEnumerator2
    Lifetime=&FloatFromCurve(
     FloatCurve=DataInterface() {propertyValues {Curve {keys [
         {arriveTangent=-5.4544, interpMode=RCIM_Cubic, leaveTangent=-5.4544, tangentMode=RCTM_User, time=0, value=0.75}
         {arriveTangent=-0.65, interpMode=RCIM_Cubic, leaveTangent=-0.65, time=0.1, value=0.4}
         <at 1 0.1>
        ]}}}
     CurveIndex=&RandomRangeFloat(Minimum=0, Maximum=1)
    )
    Sprite Size Mode=::NewEnumerator3
    Uniform Sprite Size=&Multiply_Float(A=$System.Global Scale, B=&RandomRangeFloat(Minimum=8, Maximum=14))
    UsePositionOffset=false
    Write Material Random=true
   )
   &ShapeLocation(
    MaskSpawnGroup=false
    Override Seed=false
    Rotation Axis=$System.DirectionVector
    Rotation Mode=::NewEnumerator1
    Sphere Radius=5
    UseOwnerScale=true
   )
   &CameraOffset(Camera Offset Amount=10)
   &AddVelocityFromPoint(
    Module.Use Falloff Distance=false
    Velocity Strength=&Multiply_Float(
     A=&RandomRangeFloat(Minimum=150, Maximum=200)
     B=&FloatFromCurve(FloatCurve=DataInterface() {propertyValues {Curve {keys [<at 0.7 1>,<at 0.85 3>,<at 1 3>]}}})
    )
   )
   &AddVelocity(
    Velocity=$System.DirectionVector
    Velocity Speed Scale=&Multiply_Float(A=&RandomRangeFloat(Minimum=200, Maximum=500), B=$System.Global Scale)
   )
   &DynamicMaterialParameters(
    Index 0 Param 1=3
    Index 0 Param 2=20
    Index 0 Param 3=1
    Module.Index0Float0Write=true
    Module.Index0Float1Write=true
    Module.Index0Float2Write=true
    Module.Index0Float3Write=true
   )
  ]
  particleUpdateScript [
   &ParticleState()
   &AccelerationForce(Acceleration=&RandomRangeVector(Minimum=<0 0 -2000>, Maximum=<0 0 -1000>))
   &Drag(Drag=&RandomRangeFloat(Minimum=1, Maximum=4), Use Linear Drag=true, Use Rotational Drag=true)
   &SolveForcesAndVelocity(Clamp Velocity=false, Limit Acceleration=false)
   &ScaleSpriteSizeBySpeed(Max Scale Factor=<0.1 2>, Min Scale Factor=<0.25 0.2>, Velocity Threshold=666)
   &ScaleSpriteSize(Uniform Curve Sprite Scale=DataInterface() {propertyValues {Curve {keys [{arriveTangent=-6.725, interpMode=RCIM_Cubic, leaveTangent=-6.725, tangentMode=RCTM_User, time=0.65, value=1}, <at 1 0 RCIM_Cubic>]}}})
   &ScaleColor(
    Scale Alpha=&FloatFromCurve(
     FloatCurve=DataInterface() {propertyValues {Curve {keys [
         {interpMode=RCIM_Cubic, tangentMode=RCTM_User, time=0, value=1}
         {arriveTangent=-0.3845, interpMode=RCIM_Cubic, leaveTangent=-0.3845, tangentMode=RCTM_User, time=0.5, value=0.9}
         {arriveTangent=-0.2828, interpMode=RCIM_Cubic, leaveTangent=-0.2828, tangentMode=RCTM_User, time=1, value=0}
        ]}}}
     Scale Curve=&RandomRangeFloat(Minimum=0.25, Maximum=5)
    )
    Scale RGB=&Multiply_VectorByFloat(
     Vector=&MakeVectorFromLinearColorRGB(Color=&ColorFromCurve(ColorCurve=DataInterface() {propertyValues {
         AlphaCurve {keys [<at 0 1 RCIM_Cubic>]}
         BlueCurve {keys [<at 0.1873 2 RCIM_Cubic>,<at 0.854 0.0935 RCIM_Cubic>]}
         GreenCurve {keys [<at 0.1873 2 RCIM_Cubic>,<at 0.854 0 RCIM_Cubic>]}
         RedCurve {keys [<at 0.1873 2 RCIM_Cubic>,<at 0.854 0.6406 RCIM_Cubic>]}
        }}))
     Float=20
    )
    ScaleA=true
    ScaleRGB=true
   )
  ]
  rendererClasses [&NiagaraSpriteRendererProperties]
  renderers [{rendererClass=&NiagaraSpriteRendererProperties}]
 }
 Streaks_Fast {
  emitterUpdateScript [
   &EmitterState()
   &SpawnBurst_Instantaneous(Spawn Count=2, Use Spawn Probability=true, UseLoopCountLimit=false)
  ]
  particleSpawnScript [
   &InitializeParticle(
    Color=<rgba 2 1.0162 0 1>
    Lifetime Max=0.18
    Lifetime Min=0.12
    Lifetime Mode=::NewEnumerator1
    Sprite Size=&UniformRangedVector2D(Minimum=<15 60>, Maximum=<20 80>)
    Sprite Size Mode=::NewEnumerator4
    Sprite UV Mode=::NewEnumerator1
    UsePositionOffset=false
    Write Material Random=true
   ) {module=InitializeParticle001}
   &SubUVAnimation(End Frame=3, SubUV Animation Mode:ENiagara_SubUVLookupMode=::NewEnumerator2)
   &AddVelocity(Velocity=$System.DirectionVector, Velocity Speed Scale=&Multiply_Float(A=100, B=$System.Global Scale))
  ]
  particleUpdateScript [
   &ParticleState()
   &SolveForcesAndVelocity(Clamp Velocity=false, Limit Acceleration=false)
   &ScaleSpriteSize(Scale Factor=&Vector2DFromCurve(
     Vector2Curve=DataInterface() {propertyValues {XCurve {keys [<at 0 1 RCIM_Cubic>,<at 1 0 RCIM_Cubic>]}, YCurve {keys [<at 0 0 RCIM_Cubic>,<at 1 2 RCIM_Cubic>]}}}
     ScaleCurve=&Multiply_Vector2D_ByFloat(Float=$System.Global Scale, Vector2D=<1.5 1>)
    ))
   &ScaleColor(
    Linear Color Curve=DataInterface() {propertyValues {
      AlphaCurve {keys [<at 0 1>,<at 1 0>]}
      BlueCurve {keys [<at 0 1>,<at 0.682 0>,<at 1 0.0356>]}
      GreenCurve {keys [<at 0 1>,<at 0.682 0.0974>,<at 1 0>]}
      RedCurve {keys [<at 0 1>,<at 0.682 0.4316>,<at 1 0.2031>]}
     }}
    Scale Mode=::NewEnumerator2
   )
  ]
  rendererClasses [&NiagaraSpriteRendererProperties]
  renderers [{rendererClass=&NiagaraSpriteRendererProperties}]
 }
 directional_Spikes {
  emitterSpawnScript [/SetVariables(Emitter.SpawnCount=2) {isSetParameters=true}]
  emitterUpdateScript [
   &EmitterState()
   &SpawnBurst_Instantaneous(Spawn Count=2, Use Spawn Probability=true, UseLoopCountLimit=false)
  ]
  particleSpawnScript [
   /SetVariables(Particles.Variation:NiagaraFloat=&FloatFromCurve(
     FloatCurve=DataInterface() {propertyValues {Curve {keys [{tangentMode=RCTM_User, time=0, value=0}, {tangentMode=RCTM_User, time=1, value=1}]}}}
     CurveIndex=&Divide_Float(
      A=&MakeFloatFromInt(Int=&ReturnExecIndex())
      B=&MakeFloatFromInt(Int=&Add_Integer(A=$Emitter.SpawnCount, B=-1))
     )
    )) {isSetParameters=true}
   &InitializeParticle(
    Color=&RandomRangeLinearColor(Minimum=<rgba 3 1.9964 0 1>, Maximum=<rgba 1.5 0.3186 0 1>)
    Lifetime=&Lerp_Float(A=0.3, B=0.19, Alpha=$Particles.Variation)
    Sprite Size Mode=::NewEnumerator3
    Uniform Sprite Size=&Lerp_Float(A=55, B=90, Alpha=$Particles.Variation)
    UsePositionOffset=false
    Write Material Random=true
   )
   &ShapeLocation(MaskSpawnGroup=false, Override Seed=false, Sphere Radius=6, UseOwnerScale=true)
   &AddVelocity(
    Velocity=$System.DirectionVector
    Velocity Speed Scale=&Multiply_Float(
     A=&RandomRangeFloat(Minimum=250, Maximum=400)
     B=&FN_FrontEndFloatModifier(Gameplay=1, Front End=1.2)
    )
   ) {module=AddVelocity001}
   /SetVariables(Particles.LineSubUV:NiagaraFloat=&MakeFloatFromInt(Int=&RandomRangeInt(Minimum=0, Maximum=3))) {isSetParameters=true}
  ]
  particleUpdateScript [
   &Drag(
    Drag=&FN_FrontEndFloatModifier(Gameplay=5, Front End=7)
    Use Linear Drag=true
    Use Rotational Drag=false
   )
   &SolveForcesAndVelocity(Clamp Velocity=false, Limit Acceleration=false)
   &ParticleState()
   &DynamicMaterialParameters(
    Index 0 Param 1=&FloatFromCurve(FloatCurve=DataInterface() {propertyValues {Curve {keys [
         {arriveTangent=0.8115, interpMode=RCIM_Cubic, leaveTangent=0.8115, tangentMode=RCTM_User, time=0.25, value=0.2}
         {arriveTangent=0.7921, interpMode=RCIM_Cubic, leaveTangent=0.7921, tangentMode=RCTM_User, time=1, value=1}
        ]}}})
    Index 0 Param 2=&SelectFloatFromArray(
     Array Sampling Mode=::NewEnumerator0
     Direct Array Index=?
     Float Selection Array=DataInterface() {propertyValues {FloatData [0,3]}}
     Random Seed=0
    )
    Index 1 Param 2=1
    Index 1 Param 3=1
    Index 1 Param 4=&FloatFromCurve(FloatCurve=DataInterface() {propertyValues {Curve {keys [<at 0.5 1 RCIM_Cubic>, {arriveTangent=-3.1839, tangentMode=RCTM_User, time=1, value=0.35}]}}})
    Module.Index0Float0Write=true
    Module.Index0Float1Write=true
    Module.Index0Float2Write=true
    Module.Index0Float3Write=true
    Module.Index1Float0Write=true
    Module.Index1Float1Write=true
    Module.Index1Float2Write=true
    Module.Index1Float3Write=true
    Write Parameter Index 1=::NewEnumerator1
   )
   &ScaleColor(
    Scale Alpha=&FloatFromCurve(FloatCurve=DataInterface() {propertyValues {Curve {keys [
         {arriveTangent=7.213, interpMode=RCIM_Cubic, leaveTangent=7.213, tangentMode=RCTM_User, time=0, value=0.6}
         {arriveTangentWeight=0.0568, interpMode=RCIM_Cubic, leaveTangentWeight=0.1099, tangentMode=RCTM_User, time=0.1023, value=1}
         {interpMode=RCIM_Cubic, leaveTangentWeight=0.3967, tangentMode=RCTM_User, time=0.3, value=1}
         {arriveTangent=-3.6906, interpMode=RCIM_Cubic, leaveTangent=-3.6906, tangentMode=RCTM_User, time=1, value=0}
        ]}}})
    Scale RGB=&MakeVectorFromLinearColorRGB(Color=&ColorFromCurve(ColorCurve=DataInterface() {propertyValues {
        AlphaCurve {keys [<at 0 1 RCIM_Cubic>]}
        BlueCurve {keys [<at 0.0028 4>,<at 0.5267 0>,<at 0.8778 0>]}
        GreenCurve {keys [<at 0.0028 4>,<at 0.5267 0.4056>,<at 0.8778 0.0052>]}
        RedCurve {keys [<at 0.0028 4>,<at 0.5267 0.9036>,<at 0.8778 0.1562>]}
       }}))
    ScaleA=true
    ScaleRGB=true
   )
   /SetVariables(
    Transient.DistanceScale:NiagaraFloat=&FloatFromCurve(
     FloatCurve=DataInterface() {propertyValues {Curve {keys [<at 0 1 RCIM_Cubic>, {arriveTangent=-2.2435, tangentMode=RCTM_User, time=1, value=0.1}]}}}
     CurveIndex=&Multiply_Float(A=$Particles.NormalizedAge, B=&RandomRangeFloat(Minimum=0.75, Maximum=1))
     Scale Curve=&Multiply_Float(
      A=&Multiply_Float(A=2, B=$System.Global Scale)
      B=&VectorLength(Vector=&Subtract_Vector(
        A=Linked() {linkedVariable {name=Particles.Position, type=NiagaraPosition}}
        B=Linked() {linkedVariable {name=Particles.Initial.Position, type=NiagaraPosition}}
       ))
     )
    )
    Transient.WidthScale:NiagaraFloat=&FloatFromCurve(
     FloatCurve=DataInterface() {propertyValues {Curve {keys [<at 0 0.3>,<at 1 0.2>]}}}
     Scale Curve=&Multiply_Float(A=&MakeFloatFromVector2d(Vector 2D=$Particles.Initial.SpriteSize), B=$System.Global Scale)
    )
   ) {isSetParameters=true}
   /SetVariables(
    Particles.LineColor:LinearColor=&MultiplyLinearColor(Linear Color=$Particles.Color, Scale Factor=<1 1 1 1>)
    Particles.LineScale:Vector2f=&MakeVector2D(X=$Transient.WidthScale, Y=$Transient.DistanceScale)
   ) {isSetParameters=true}
   &SpriteFacingAndAlignment(
    Sprite Alignment=&Multiply_VectorByFloat(Vector=$Particles.Velocity, Float=-1)
    WriteAlignment=true
    WriteFacing=false
   )
  ]
  rendererClasses [&NiagaraSpriteRendererProperties]
  renderers [{rendererClass=&NiagaraSpriteRendererProperties}]
 }
}
summary {
 systemName=NS_Pickaxe_Bass_GatorCeladon_Impact
 emitters [
  {emitterName=Impact_Smoke, rendererClasses [&NiagaraSpriteRendererProperties]}
  {
   emitterName=FN_PhysMatDebris001
   rendererClasses [&NiagaraSpriteRendererProperties, &NiagaraMeshRendererProperties]
  }
  {emitterName=Flash_Flare, rendererClasses [&NiagaraSpriteRendererProperties]}
  {emitterName=directional_Spikes, rendererClasses [&NiagaraSpriteRendererProperties]}
  {emitterName=Sparks, rendererClasses [&NiagaraSpriteRendererProperties]}
  {
   emitterName=Flash_Flare_Light
   rendererClasses [&NiagaraSpriteRendererProperties, &NiagaraLightRendererProperties]
  }
  {emitterName=Streaks_Fast, rendererClasses [&NiagaraSpriteRendererProperties]}
  {bEnabled=false, emitterName=DirectionalSparks001, rendererClasses [&NiagaraSpriteRendererProperties]}
  {emitterName=Flash_Combined, rendererClasses [&NiagaraSpriteRendererProperties, &NiagaraMeshRendererProperties]}
 ]
 userVariables [
  User.bIsFrontEnd {type=NiagaraBool}
  User.bIsFrontEndPreview {type=NiagaraBool}
  User.Direction {type=Vector3f, description="", defaultValue {struct=Vector3f, value=<0 1 0>}}
  User.FrontEndScale {type=NiagaraFloat, defaultValue=0.35, description=""}
  User.PhysicalSurface {type=NiagaraInt32}
  User.SwingSectionIndex {type=NiagaraInt32}
 ]
}
